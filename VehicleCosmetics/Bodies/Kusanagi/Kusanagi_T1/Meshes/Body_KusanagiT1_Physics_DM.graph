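# Physics asset for the Kusanagi T1 vehicle body cosmetic, providing simplified collision for its skeletal mesh.
# A single box shape on the root bone covers the whole body, offset slightly forward and raised above the origin.
# With one body and no constraints, it acts as one rigid volume with no articulated parts.
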PhysicsAsset /VehicleCosmetics/Bodies/Kusanagi/Kusanagi_T1/Meshes/Body_KusanagiT1_Physics_DM
bodies {root [{shapeType=Box, extentX=444, extentY=216, extentZ=116, shapeName=root_box, center=<25 0 106>}]}
constraints []
